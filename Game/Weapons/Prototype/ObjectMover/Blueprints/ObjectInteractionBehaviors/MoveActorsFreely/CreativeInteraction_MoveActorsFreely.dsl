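; Creative move tool interaction behavior for freely moving actors, driving preview transition effects and a placement sound.
; Starting interaction on the server spawns TransitionIn previews for selected actors; ending it, where cosmetics cannot execute, sets their previews to RevertPlacement and backwards playback.
; Placement spawns a TransitionOut preview at the original actor's transform, or the spawned one's if invalid.
; Unreplicated mesh components fall back to the spawned actor's own.
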(event CreativeMoveTool|EventEndCreativeInteraction
  (bind _returnvalue (CreativeInteraction|GetFortCreativeMoveTool))
  ;[1 1 1] End Creative Interaction, Return Held object
  (|CallParentFunction)
  (Networking|CanExecuteCosmeticEvents
    (:False
      (for _array_element (CreativeMoveTool|GetOwnerSelectedActors)
        (Utilities|IsValid (Class|FortCreativeMoveTool|FindActivePreviewforObject _returnvalue _array_element)
          (:Is%20Valid
            (Class|CreativeMoveToolBlueprintPreview|ServerSetFXState (Class|FortCreativeMoveTool|FindActivePreviewforObject _returnvalue _array_element) "RevertPlacement")
            (Class|CreativeMoveToolBlueprintPreview|ServerSetPlayTransitionBackwardswhenCurrentAnimEnds (Class|FortCreativeMoveTool|FindActivePreviewforObject _returnvalue _array_element))))))))

(event Custom|SpawnedObjectEffects
  (params (Components :type Array%20of%20Mesh%20Component%20Object%20References) (ActorWorldTransform :type Transform) (SpawnedActor :type Actor%20Object%20Reference))
  (bind _returnvalue (Actor|GetComponentsByClass SpawnedActor "/Script/Engine.MeshComponent"))
  (bind _output (Utilities|Array|Get(acopy) Components))
  (bind _output_2 (Utilities|Array|Get(acopy) _returnvalue))
  ;[1 1 1] Server Create Spawn effects
  (if (Utilities|Array|IsNotEmpty Components)
    (Utilities|IsValid _output
      (:Is%20Valid
        _)
      (:Is%20Not%20Valid
        ;;[1 0.7787 0.4828] Fallback for props that have unreplicated mesh components
        (Utilities|IsValid SpawnedActor
          (:Is%20Valid
            (if (Utilities|Array|IsNotEmpty _returnvalue)
              (Utilities|IsValid _output_2
                (:Is%20Valid
                  _)))))))
    (bind _data (Utilities|FlowControl|MergeExecutionwithData _output _output_2)
      (:Then
        (Utilities|IsValid (Components|GetOwner _data)
          (:Is%20Valid
            (CallFunction|SpawnPreviewActor (Components|GetOwner _data) "TransitionOut" ActorWorldTransform)))))
    (else
      _)))

(event CreativeMoveTool|EventOnObjectPlacement
  (params (Components :type Array%20of%20Mesh%20Component%20Object%20References) (NewlyPlacedActorPair :type Original%20And%20Spawned%20Pair%20Structure))
  (bind (_originalactor _spawnedactor) (Utilities|Struct|BreakOriginalandSpawnedPair NewlyPlacedActorPair))
  ;[1 1 1] On duplicate For the new actor. Currently Client only
  (Utilities|IsValid _originalactor
    (:Is%20Valid
      (CallFunction|SpawnedObjectEffects Components (Transformation|GetActorTransform _originalactor) _spawnedactor))
    (:Is%20Not%20Valid
      (CallFunction|SpawnedObjectEffects Components (Transformation|GetActorTransform _spawnedactor) _spawnedactor))))

(event CreativeMoveTool|EventStartCreativeInteractionOnServer
  ;[1 1 1] Server Start Interacting (first thing that happens on cut or copy)
  (|CallParentFunction)
  (for _array_element (CreativeMoveTool|GetOwnerSelectedActors)
    (CallFunction|SpawnPreviewActor _array_element "TransitionIn")))

(event CreativeMoveTool|EventTriggerInteractionBehavior
  ;[1 1 1] Triggered
  (|CallParentFunction)
  (Audio|PlaySoundatLocation "/Game/Sounds/Creative/Gadgets/MoveTool/MoveTool_Place_01_Cue.MoveTool_Place_01_Cue" (Transformation|GetActorLocation (Components|GetOwner))))

(fn SpawnPreviewActor
  (params (PropActor :type Actor%20Object%20Reference) (SpawnTransitionState :type ECreativeMoveToolPreviewFX%20Enum) (ActorWorldTransform :type Transform))
  ;[1 1 1] Spawn a new preview actor
  (Utilities|IsValid PropActor
    (:Is%20Valid
      (CallFunction|SpawnPreview PropActor SpawnTransitionState ActorWorldTransform)
      (return))))
